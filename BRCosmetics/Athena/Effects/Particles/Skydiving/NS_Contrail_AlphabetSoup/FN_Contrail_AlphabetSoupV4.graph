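# Niagara system for the Alphabet Soup skydiving contrail cosmetic, attached to bones or sockets on the user-supplied skeletal mesh.
# Two letter emitters fire timed single bursts, choosing flipbook frames from emitter-age curves for ordered sequences, while a third scatters random glyphs.
# A ribbon trail and two tinted soup-bubble sprite emitters form the broth.
# Front-end modifiers adjust motion for menu previews, and most emitters add scaled, inverted owner velocity for drag-behind.
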NiagaraSystem /BRCosmetics/Athena/Effects/Particles/Skydiving/NS_Contrail_AlphabetSoup/FN_Contrail_AlphabetSoupV4
dependencies {
 usedDataInterfaces [
  &NiagaraDataInterfaceVector2DCurve
  &NiagaraDataInterfaceCurve
  &NiagaraDataInterfaceVectorCurve
 ]
 usedDynamicInputs []
 usedModules [&SystemState]
}
emitters {
 FN_Contrails_Letters_A {
  emitterSpawnScript [&FN_PreserveSkydiveUserVariables() {module=PreserveUserVariables}]
  emitterUpdateScript [
   &EmitterState()
   &SpawnBurst_Instantaneous(Spawn Count=1, Spawn Time=0.1936, Use Spawn Probability=false)
   &SpawnBurst_Instantaneous(Spawn Count=1, Spawn Time=0.3375, Use Spawn Probability=false) {module=SpawnBurst_Instantaneous001}
   &SpawnBurst_Instantaneous(Spawn Count=1, Spawn Time=0.498, Use Spawn Probability=false) {module=SpawnBurst_Instantaneous002}
   &SpawnBurst_Instantaneous(Spawn Count=1, Spawn Time=0.6947, Use Spawn Probability=false) {module=SpawnBurst_Instantaneous003}
   &SpawnBurst_Instantaneous(Spawn Count=1, Spawn Time=0.8848, Use Spawn Probability=false) {module=SpawnBurst_Instantaneous004}
   &SpawnBurst_Instantaneous(Spawn Count=1, Spawn Time=4, Use Spawn Probability=false) {module=SpawnBurst_Instantaneous005}
   &SpawnBurst_Instantaneous(Spawn Count=1, Spawn Time=4.1, Use Spawn Probability=false) {module=SpawnBurst_Instantaneous006}
  ]
  particleSpawnScript [
   /Game/Effects/Niagara/Modules/Groups/FN_SetSpawnGroup_v1/FN_SetSpawnGroup(Emitter Bone Count=1) {enabled=false}
   /Niagara/Modules/Spawn/Initialization/InitializeParticle(
    Color=<rgba 0.7934 0.5816 0.1604 1>
    Lifetime=1.8
    Module.Write Color=true
    Module.Write Lifetime=true
    Module.Write Mass=true
    Module.Write Material Random=true
    Module.Write Position=true
    Module.Write Scale=false
    Module.Write SpriteRotation=true
    Module.Write SpriteSize=true
    Position=&SimulationPosition()
    Sprite Rotation=&UniformRangedFloat(Minimum=-10, Maximum=25)
    Sprite Size=&Vector2DFromFloat(Value=&UniformRangedFloat(Minimum=15, Maximum=16))
   )
   &SampleSkeletalMeshSkeleton(Mesh:NiagaraDataInterfaceSkeletalMesh=$User.SkeletalMesh, Specific Bone Index=1) {enabled=false}
   /Niagara/Modules/Spawn/Location/SkeletalMeshSkeletonLocation(Module.Inherit Velocity=false, Module.Use Sampled Bone as Ribbon ID=true) {enabled=false}
   &SkeletalMeshLocation(Bone / Socket Index=1, Bone Sampling Mode=::NewEnumerator4, Inherit Velocity=false, MaskSpawnGroup=false, Skeletal Mesh=$User.SkeletalMesh)
   &AddVelocity(Velocity=&FN_FrontEndVector3Modifier(Gameplay=<0 0 468>, Front End=<0 0 300>))
   &DynamicMaterialParameters(
    Index 0 Param 1=&UniformRangedFloat(Minimum=0, Maximum=1)
    Module.Param0WriteEnabled=true
    Module.Param1WriteEnabled=true
    Module.Param2WriteEnabled=true
    Module.Param3WriteEnabled=true
   )
   /SetVariables(Particles.SubImageIndex=&FloatFromCurve(
     FloatCurve=DataInterface() {propertyValues {Curve {keys [<at -0.0046 1 RCIM_Constant>,<at 0.29 1 RCIM_Constant>,<at 0.43 2 RCIM_Constant>,<at 0.59 3 RCIM_Constant>,<at 0.8 3 RCIM_Constant>,<at 3.9 13 RCIM_Constant>,<at 4.05 14 RCIM_Constant>,<at 5 16 RCIM_Constant>]}}}
     CurveIndex=$Emitter.LoopedAge
    )) {isSetParameters=true}
  ]
  particleUpdateScript [
   &ParticleState()
   &AccelerationForce(
    Acceleration=&VectorFromCurve(VectorCurve=DataInterface() {propertyValues {
       XCurve {keys [<at 0.0017 -132>,<at 0.9905 65>]}
       YCurve {keys [<at -0.0011 1.0067>,<at 0.9988 0>]}
       ZCurve {keys [<at -0.0011 -300>,<at 0.9957 -200.05>]}
      }})
    Coordinate Space=::Local
   )
   &Drag(Use Linear Drag=true, Use Rotational Drag=false)
   &InheritVelocity(
    Module.Limit Speed=true
    Velocity Limit=1000
    Velocity Scale=<-0.005 -0.005 0>
    Velocity Source=&TransformVector(Vector=$Engine.Owner.Velocity, Source Space=::World, Destination Space=::Local)
   )
   &SolveForcesAndVelocity(Clamp Velocity=false, Limit Acceleration=false)
   &ScaleSpriteSize(Scale Factor=&Vector2DFromFloat(Value=&FloatFromCurve(FloatCurve=DataInterface() {propertyValues {Curve {keys [<at 0 0>,<at 0.0689 0.9252>,<at 0.9242 1.0835>,<at 1.0011 0.002>]}}})))
  ]
  rendererClasses [&NiagaraSpriteRendererProperties]
  renderers [{rendererClass=&NiagaraSpriteRendererProperties}]
 }
 FN_Contrails_Letters_B {
  emitterSpawnScript [&FN_PreserveSkydiveUserVariables() {module=PreserveUserVariables}]
  emitterUpdateScript [
   &EmitterState()
   &SpawnBurst_Instantaneous(Spawn Count=1, Spawn Time=2.2741, Use Spawn Probability=false)
   &SpawnBurst_Instantaneous(Spawn Count=1, Spawn Time=2.401, Use Spawn Probability=false) {module=SpawnBurst_Instantaneous001}
   &SpawnBurst_Instantaneous(Spawn Count=1, Spawn Time=2.5126, Use Spawn Probability=false) {module=SpawnBurst_Instantaneous002}
   &SpawnBurst_Instantaneous(Spawn Count=1, Spawn Time=2.6289, Use Spawn Probability=false) {module=SpawnBurst_Instantaneous003}
   &SpawnBurst_Instantaneous(Spawn Count=1, Spawn Time=2.7371, Use Spawn Probability=false) {module=SpawnBurst_Instantaneous004}
   &SpawnBurst_Instantaneous(Spawn Count=1, Spawn Time=2.8704, Use Spawn Probability=false) {module=SpawnBurst_Instantaneous005}
   &SpawnBurst_Instantaneous(Spawn Count=1, Spawn Time=2.9864, Use Spawn Probability=false) {module=SpawnBurst_Instantaneous006}
   &SpawnBurst_Instantaneous(Spawn Count=1, Spawn Time=3.1032, Use Spawn Probability=false) {module=SpawnBurst_Instantaneous007}
  ]
  particleSpawnScript [
   &FN_SetSpawnGroup(Emitter Bone Count=1) {enabled=false}
   &FN_SetSpawnGroup_v2(Number of Spawn Groups=1)
   &InitializeParticle(
    Color=<rgba 0.7934 0.5816 0.1604 1>
    Lifetime=1.8
    Module.Write Color=true
    Module.Write Lifetime=true
    Module.Write Mass=true
    Module.Write Material Random=true
    Module.Write Position=false
    Module.Write Scale=false
    Module.Write SpriteRotation=true
    Module.Write SpriteSize=true
    Position=&SimulationPosition()
    Sprite Rotation=&UniformRangedFloat(Minimum=-10, Maximum=25)
    Sprite Size=&Vector2DFromFloat(Value=&UniformRangedFloat(Minimum=15, Maximum=16))
   )
   &SampleSkeletalMeshSkeleton(Mesh:NiagaraDataInterfaceSkeletalMesh=$User.SkeletalMesh, Specific Bone Index=0) {enabled=false}
   &SkeletalMeshSkeletonLocation(Module.Inherit Velocity=false, Module.Use Sampled Bone as Ribbon ID=true) {enabled=false}
   &SkeletalMeshLocation(
    Bone / Socket Index=$Particles.SpawnGroup
    Bone Sampling Mode=::NewEnumerator4
    Inherit Velocity=false
    MaskSpawnGroup=false
    Skeletal Mesh=$User.SkeletalMesh
   )
   &DynamicMaterialParameters(
    Index 0 Param 1=&UniformRangedFloat(Minimum=0, Maximum=1)
    Module.Param0WriteEnabled=true
    Module.Param1WriteEnabled=true
    Module.Param2WriteEnabled=true
    Module.Param3WriteEnabled=true
   )
   /SetVariables(Particles.SubImageIndex=&FloatFromCurve(
     FloatCurve=DataInterface() {propertyValues {Curve {keys [<at 0.131 27.0333>,<at 1.6115 5>,<at 2 12 RCIM_Constant>,<at 2.35 11 RCIM_Constant>,<at 2.48 10 RCIM_Constant>,<at 2.6 9 RCIM_Constant>,<at 2.7 8 RCIM_Constant>,<at 2.85 7 RCIM_Constant>,<at 2.95 6 RCIM_Constant>,<at 3.1 5>]}}}
     CurveIndex=$Emitter.LoopedAge
    )) {isSetParameters=true}
   &AddVelocity(Velocity=&FN_FrontEndVector3Modifier(Gameplay=<0 0 458>, Front End=<0 0 300>))
   &ApplyInitialForces()
  ]
  particleUpdateScript [
   /Niagara/Modules/Update/Forces/CurlNoiseForce(
    Noise Strength=&FN_FrontEndFloatModifier(
     Gameplay=0
     Front End=&UniformRangedFloat(
      Minimum=&FloatFromCurve(FloatCurve=DataInterface() {propertyValues {Curve {keys [<at 0.5 0>,<at 1.0011 1000>]}}})
      Maximum=&FloatFromCurve(FloatCurve=DataInterface() {propertyValues {Curve {keys [<at 0.5 0>,<at 1 -1000>]}}})
     )
    )
    Pan Noise=false
   )
   &ParticleState()
   &AccelerationForce(
    Acceleration=&VectorFromCurve(VectorCurve=DataInterface() {propertyValues {
       XCurve {keys [<at 0.0017 132>,<at 0.9905 -65>]}
       YCurve {keys [<at -0.0011 1.0067>,<at 0.9988 0>]}
       ZCurve {keys [<at -0.0011 -300>,<at 0.9957 -200.05>]}
      }})
    Coordinate Space=::Local
   )
   &Drag(Use Linear Drag=true, Use Rotational Drag=false)
   &InheritVelocity(
    Module.Limit Speed=true
    Velocity Limit=1000
    Velocity Scale=<-0.005 -0.005 0>
    Velocity Source=&TransformVector(Vector=$Engine.Owner.Velocity, Source Space=::World, Destination Space=::Local)
   )
   &SolveForcesAndVelocity(Clamp Velocity=false, Limit Acceleration=false)
   &ScaleSpriteSize(Scale Factor=&Vector2DFromFloat(Value=&FloatFromCurve(FloatCurve=DataInterface() {propertyValues {Curve {keys [<at -0.0014 -0.0319>,<at 0.1056 1.0068>,<at 0.8891 0.9859>,<at 1.0011 0.002>]}}})))
  ]
  rendererClasses [&NiagaraSpriteRendererProperties]
  renderers [{rendererClass=&NiagaraSpriteRendererProperties}]
 }
 FN_Contrails_Letters_C_Random001 {
  emitterSpawnScript [&FN_PreserveSkydiveUserVariables() {module=PreserveUserVariables}]
  emitterUpdateScript [
   &EmitterState(
    Inactive Response=::NewEnumerator1
    Life Cycle Mode=::NewEnumerator1
    Loop Delay=&UniformRangedFloat(Minimum=0.2, Maximum=0.4)
    Loop Duration=&UniformRangedFloat(Minimum=0.2, Maximum=1)
    Recalculate Duration Each Loop=true
    UseLoopDelay=true
   )
   /Game/Effects/Niagara/Modules/Spawn/SpawnRateFromMultipleSpawnGroups_2(Global Spawnrate:NiagaraFloat=&UniformRangedFloat(Minimum=6, Maximum=12))
  ]
  particleSpawnScript [
   &FN_SetSpawnGroup(Emitter Bone Count=2) {enabled=false}
   &FN_SetSpawnGroup_v2(Number of Spawn Groups=2)
   &InitializeParticle(
    Color=<rgba 0.7934 0.5816 0.1604 1>
    Lifetime=1.8
    Module.Write Color=true
    Module.Write Lifetime=true
    Module.Write Mass=true
    Module.Write Material Random=true
    Module.Write Position=false
    Module.Write Scale=false
    Module.Write SpriteRotation=true
    Module.Write SpriteSize=true
    Position=&SimulationPosition()
    Sprite Rotation=&UniformRangedFloat(Minimum=-10, Maximum=25)
    Sprite Size=&Vector2DFromFloat(Value=&UniformRangedFloat(Minimum=5, Maximum=15))
   )
   &SampleSkeletalMeshSkeleton(Mesh:NiagaraDataInterfaceSkeletalMesh=$User.SkeletalMesh, Specific Bone Index:NiagaraInt32=$Particles.SpawnGroup) {enabled=false}
   &SkeletalMeshSkeletonLocation(Module.Inherit Velocity=false, Module.Use Sampled Bone as Ribbon ID=true) {enabled=false}
   &SkeletalMeshLocation(
    Bone / Socket Index=$Particles.SpawnGroup
    Bone Sampling Mode=::NewEnumerator4
    Inherit Velocity=false
    MaskSpawnGroup=false
    Skeletal Mesh=$User.SkeletalMesh
   )
   &AddVelocity(Velocity=&FN_FrontEndVector3Modifier(
     Gameplay=&UniformRangedVector(Minimum=<-50 -50 350>, Maximum=<50 50 450>)
     Front End=&UniformRangedVector(Minimum=<-50 -50 400>, Maximum=<50 50 400>)
    ))
   &DynamicMaterialParameters(
    Index 0 Param 1=&UniformRangedFloat(Minimum=0, Maximum=1)
    Module.Param0WriteEnabled=true
    Module.Param1WriteEnabled=true
    Module.Param2WriteEnabled=true
    Module.Param3WriteEnabled=true
   )
   /SetVariables(Particles.RandomFloat=&UniformRangedFloat(Minimum=-1, Maximum=1)) {isSetParameters=true}
   /SetVariables(Particles.SubImageIndex=&UniformRangedFloat(Minimum=0, Maximum=15)) {isSetParameters=true}
   /Niagara/Modules/Spawn/Location/SphereLocation(MaskSpawnGroup=false, Module.Override Randomness=false, Sphere Radius=15, Surface Only=false)
  ]
  particleUpdateScript [
   &ParticleState()
   &SpriteRotationRate(Delta Time:NiagaraFloat=$Particles.RandomFloat, Rotation Rate=10)
   &AccelerationForce(Acceleration=<0 0 -780>)
   &Drag(Use Linear Drag=true, Use Rotational Drag=false)
   &InheritVelocity(
    Module.Limit Speed=true
    Velocity Limit=1000
    Velocity Scale=<-0.01 -0.01 -0.01>
    Velocity Source=&TransformVector(Vector=$Engine.Owner.Velocity, Source Space=::World, Destination Space=::Local)
   )
   &SolveForcesAndVelocity(Clamp Velocity=false, Limit Acceleration=false)
   &ScaleSpriteSize(Scale Factor=&Vector2DFromFloat(Value=&FloatFromCurve(FloatCurve=DataInterface() {propertyValues {Curve {keys [<at -0.0014 -0.0319>,<at 0.1056 1.0068>,<at 0.8891 0.9859>,<at 1.0011 0.002>]}}})))
  ]
  rendererClasses [&NiagaraSpriteRendererProperties]
  renderers [{rendererClass=&NiagaraSpriteRendererProperties}]
 }
 FN_Contrails_Ribbon_v4_A {
  emitterUpdateScript [
   &EmitterScalability(
    Module.bApplySpawnCountScale=false
    Module.bApplySpawnCountScaleByDistanceFraction=false
    Module.bUseMaxDistance=false
    Module.bUseMinDistance=false
   ) {enabled=false}
   &EmitterLifeCycle(CompleteOnInactive=true) {enabled=false}
   &EmitterState()
   /Game/Effects/Niagara/Modules/Spawn/FN_SpawnContrails(All limbs=false, Global Spawnrate=60) {enabled=false}
   /Game/Effects/Niagara/Modules/AnimTrail/FN_SpawnAnimTrail/FN_SpawnAnimTrail() {enabled=false}
   &SpawnRate(SpawnRate=30, Use Spawn Probability=false) {enabled=false}
   &FN_SpawnRibbon_v2(Number of Ribbons=2, Spawn Mode=::NewEnumerator0)
  ]
  particleSpawnScript [
   &FN_InitializeRibbon_v3(Enable Skeletal Position=true, Lifetime=0.7, Ribbon Width=60) {enabled=false}
   &FN_InitializeRibbon_v3(
    Enable Skeletal Position=true
    Lifetime=0.7
    Ribbon Width=60
    Sampled Position Coordinate Space:ENiagaraCoordinateSpace=::Local
    Sampled Rotation Coordinate Space:ENiagaraCoordinateSpace=::Local
    Sampled Velocity Coordinate Space:ENiagaraCoordinateSpace=::Local
   ) {module=FN_InitializeRibbon_v3001}
   &AddVelocity(Velocity=<0 0 410>) {enabled=false}
   &AddVelocity(Velocity=&MakeCustomVectorFromBool(
     Bool=$Particles.Locked
     True Vector=<0 0 0>
     False Vector=&FN_FrontEndVector3Modifier(Gameplay=<0 0 410>, Front End=<0 0 350>)
    )) {module=AddVelocity001}
  ]
  particleUpdateScript [
   &AccelerationForce(
    Acceleration=&VectorFromCurve(VectorCurve=DataInterface() {propertyValues {
       XCurve {keys [<at 0.0017 132.1921>,<at 0.9905 -65.3336>]}
       YCurve {keys [<at -0.0011 1.0067>,<at 0.9988 0>]}
       ZCurve {keys [<at -0.0011 -300>,<at 0.9957 -200.05>]}
      }})
    Coordinate Space=::Local
   ) {enabled=false}
   &Drag(Drag=0.9, Use Linear Drag=true, Use Rotational Drag=false)
   &InheritVelocity(
    Module.Limit Speed=true
    Velocity Limit=10000
    Velocity Scale=&MakeCustomVectorFromBool(Bool=$Particles.Initial.Locked, True Vector=<0 0 0>, False Vector=<-0.03 -0.03 -0.01>)
    Velocity Source=&TransformVector(Vector=$Engine.Owner.Velocity, Source Space=::World, Destination Space=::Local)
   )
   /Game/Effects/Niagara/Modules/Skydiving/FN_DivingAttributes/FN_DivingAttributes() {enabled=false}
   &SolveForcesAndVelocity(Clamp Velocity=false, Limit Acceleration=false)
   &FN_UpdateRibbon(
    Enable Skeletal Position=true
    Initial Color:LinearColor=$Particles.Initial.Color
    Initial RibbonWidth:NiagaraFloat=$Particles.Initial.RibbonWidth
    Sampled Position Coordinate Space:ENiagaraCoordinateSpace=::Local
    Sampled Rotation Coordinate Space:ENiagaraCoordinateSpace=::Local
    Sampled Velocity Coordinate Space:ENiagaraCoordinateSpace=::Local
   )
   &Material(Base Texture Time:NiagaraFloat=&Multiply_Float(A=$Emitter.Age, B=-1))
   &DynamicMaterialParameters(Module.Param0WriteEnabled=true, Module.Param1WriteEnabled=true, Module.Param2WriteEnabled=true, Module.Param3WriteEnabled=true)
   &ScaleRibbonWidth(Ribbon Width Scale=&FloatFromCurve(FloatCurve=DataInterface() {propertyValues {Curve {keys [
         {arriveTangent=1.9737, interpMode=RCIM_Cubic, leaveTangent=1.9737, tangentMode=RCTM_User, time=0.0058, value=0.9363}
         {arriveTangent=-0.0131, interpMode=RCIM_Cubic, leaveTangent=-0.0131, time=0.4538, value=1.5093}
         <at 0.6635 0.9277>
        ]}}}))
   &Color(
    Color=<rgba 1 1 1 0.8>
    Scale Alpha=&FloatFromCurve(
     FloatCurve=DataInterface() {propertyValues {Curve {keys [
         {arriveTangent=2.7939, interpMode=RCIM_Cubic, leaveTangent=2.7939, tangentMode=RCTM_User, time=-0.0046, value=1.0113}
         {arriveTangent=-1.0036, interpMode=RCIM_Cubic, leaveTangent=-4.3444, tangentMode=RCTM_Break, time=0.1387, value=1.232}
         <at 0.6 0.002>
        ]}}}
     Scale Curve=0.8
    )
   )
  ]
  rendererClasses [&NiagaraRibbonRendererProperties]
  renderers [{rendererClass=&NiagaraRibbonRendererProperties}]
 }
 FN_Contrails_SoupBubbles {
  emitterSpawnScript [&FN_PreserveSkydiveUserVariables() {module=PreserveUserVariables}]
  emitterUpdateScript [
   &EmitterState()
   &SpawnRateFromMultipleSpawnGroups_2(Global Spawnrate=25)
  ]
  particleSpawnScript [
   &FN_SetSpawnGroup(Emitter Bone Count=2) {enabled=false}
   &FN_SetSpawnGroup_v2(Number of Spawn Groups=2)
   &InitializeParticle(
    Color=<rgba 1 1 1 0.3>
    Lifetime=&UniformRangedFloat(Minimum=0.15, Maximum=0.35)
    Module.Write Color=true
    Module.Write Lifetime=true
    Module.Write Mass=true
    Module.Write Material Random=true
    Module.Write Position=false
    Module.Write Scale=false
    Module.Write SpriteRotation=true
    Module.Write SpriteSize=true
    Position=&SimulationPosition()
    Sprite Rotation=&UniformRangedFloat(Minimum=0, Maximum=360)
    Sprite Size=<8 22>
   )
   &SampleSkeletalMeshSkeleton(Mesh:NiagaraDataInterfaceSkeletalMesh=$User.SkeletalMesh, Specific Bone Index:NiagaraInt32=$Particles.SpawnGroup) {enabled=false}
   &SkeletalMeshSkeletonLocation(Module.Inherit Velocity=false, Module.Use Sampled Bone as Ribbon ID=true) {enabled=false}
   &SkeletalMeshLocation(
    Bone / Socket Index=$Particles.SpawnGroup
    Bone Sampling Mode=::NewEnumerator4
    Inherit Velocity=false
    MaskSpawnGroup=false
    Skeletal Mesh=$User.SkeletalMesh
   )
   &SphereLocation(Hemisphere Z=true, MaskSpawnGroup=false, Module.Override Randomness=false, Sphere Radius=18, Surface Only=false)
   &AddVelocityFromPoint(
    Module.Use Falloff Distance=false
    Velocity Origin:Vector3f=&SimulationPosition()
    Velocity Strength=&UniformRangedFloat(Minimum=100, Maximum=250)
   )
   &AddVelocity(Velocity=<0 0 50>)
  ]
  particleUpdateScript [
   &ParticleState()
   &SolveForcesAndVelocity(Clamp Velocity=false, Limit Acceleration=false)
   &ScaleSpriteSize(Scale Factor=&Vector2DFromCurve(
     Vector2Curve=DataInterface() {propertyValues {XCurve {keys [<at -0.0011 1.0067>,<at 0.9997 0.4319>]}, YCurve {keys [<at 0.0102 1.1102>,<at 0.0596 0.8185>,<at 0.9997 0.3098>]}}}
     ScaleCurve=<1 1>
    ))
   &Color(
    Color=<rgba 1 0.1431 0 0.1>
    Scale Alpha=&FloatFromCurve(FloatCurve=DataInterface() {propertyValues {Curve {keys [<at 0.0048 0.0237>,<at 0.4463 0.9674>,<at 1.0011 0.002>]}}})
    Scale Color=&UniformRangedVector(Minimum=<1 2 1>, Maximum=<2 1 1>)
   )
  ]
  rendererClasses [&NiagaraSpriteRendererProperties]
  renderers [{rendererClass=&NiagaraSpriteRendererProperties}]
 }
 FN_Contrails_SoupBubbles001 {
  emitterSpawnScript [&FN_PreserveSkydiveUserVariables() {module=PreserveUserVariables}]
  emitterUpdateScript [
   &EmitterState()
   &SpawnRateFromMultipleSpawnGroups_2(Global Spawnrate=25)
  ]
  particleSpawnScript [
   &FN_SetSpawnGroup(Emitter Bone Count=2) {enabled=false}
   &FN_SetSpawnGroup_v2(Number of Spawn Groups=2)
   &InitializeParticle(
    Color=<rgba 1 0.5167 0.1567 0.89>
    Lifetime=&UniformRangedFloat(Minimum=0.2, Maximum=0.65)
    Module.Write Color=true
    Module.Write Lifetime=true
    Module.Write Mass=true
    Module.Write Material Random=true
    Module.Write Position=false
    Module.Write Scale=false
    Module.Write SpriteRotation=true
    Module.Write SpriteSize=true
    Position=&SimulationPosition()
    Sprite Rotation=&UniformRangedFloat(Minimum=0, Maximum=360)
    Sprite Size=&UniformRangedVector2D(Minimum=<9 22>, Maximum=<7 28>)
   )
   &SampleSkeletalMeshSkeleton(Mesh:NiagaraDataInterfaceSkeletalMesh=$User.SkeletalMesh, Specific Bone Index:NiagaraInt32=$Particles.SpawnGroup) {enabled=false}
   &SkeletalMeshSkeletonLocation(
    Module.Inherit Velocity=false
    Module.Use Sampled Bone as Ribbon ID=true
    Sampled Position Offset:Vector3f=&UniformRangedVector(Minimum=<0 0 0>, Maximum=<1 1 150>)
   ) {enabled=false}
   &SkeletalMeshLocation(
    Bone / Socket Index=$Particles.SpawnGroup
    Bone Sampling Mode=::NewEnumerator4
    Inherit Velocity=false
    MaskSpawnGroup=false
    Sampled Position Offset=&UniformRangedVector(Minimum=<-1 -1 0>, Maximum=<1 1 150>)
    Skeletal Mesh=$User.SkeletalMesh
   )
   &SphereLocation(Hemisphere Z=true, MaskSpawnGroup=false, Module.Override Randomness=false, Sphere Radius=30, Surface Only=false)
   &AddVelocityFromPoint(
    Module.Use Falloff Distance=false
    Velocity Origin:Vector3f=&SimulationPosition()
    Velocity Strength=&UniformRangedFloat(Minimum=100, Maximum=250)
   )
   &AddVelocity(Velocity=&UniformRangedVector(Minimum=<0 0 0>, Maximum=<250 250 500>))
  ]
  particleUpdateScript [
   &ParticleState()
   &AccelerationForce(Acceleration=<0 0 -2000>, Coordinate Space=::Local)
   &Drag(Drag=7, Use Linear Drag=true, Use Rotational Drag=false)
   &InheritVelocity(
    Module.Limit Speed=true
    Velocity Scale=<-0.05 -0.05 -0.03>
    Velocity Source=&TransformVector(Vector=$Engine.Owner.Velocity, Source Space=::World, Destination Space=::Local)
   )
   &SolveForcesAndVelocity(Clamp Velocity=false, Limit Acceleration=false)
   &ScaleSpriteSize(Scale Factor=&Vector2DFromCurve(
     Vector2Curve=DataInterface() {propertyValues {
       XCurve {keys [
         <at 0.0169 0.5318 RCIM_Cubic>
         {arriveTangent=1.4987, interpMode=RCIM_Cubic, leaveTangent=-2.3772, tangentMode=RCTM_Break, time=0.3344, value=0.9031}
         <at 0.9882 0.3781>
        ]}
       YCurve {keys [<at 0.0019 1.5688>, {arriveTangent=-1.3092, interpMode=RCIM_Cubic, leaveTangent=-1.3092, time=0.0393, value=0.8775}, <at 0.9994 0.2629>]}
      }}
     ScaleCurve=<1 1>
    ))
   &Color(
    Color=<rgba 1 0.1431 0 0.1>
    Scale Alpha=&FloatFromCurve(FloatCurve=DataInterface() {propertyValues {Curve {keys [<at 0.0048 0.0237>,<at 0.4463 0.9674>,<at 1.0011 0.002>]}}})
    Scale Color=&UniformRangedVector(Minimum=<1 2 1>, Maximum=<2 1 1>)
   )
  ]
  rendererClasses [&NiagaraSpriteRendererProperties]
  renderers [{rendererClass=&NiagaraSpriteRendererProperties}]
 }
}
summary {
 systemName=FN_Contrail_AlphabetSoupV4
 emitters [
  {emitterName=FN_Contrails_SoupBubbles, rendererClasses [&NiagaraSpriteRendererProperties]}
  {emitterName=FN_Contrails_Letters_B, rendererClasses [&NiagaraSpriteRendererProperties]}
  {emitterName=FN_Contrails_Letters_A, rendererClasses [&NiagaraSpriteRendererProperties]}
  {emitterName=FN_Contrails_Ribbon_v4_A, rendererClasses [&NiagaraRibbonRendererProperties]}
  {emitterName=FN_Contrails_Letters_C_Random001, rendererClasses [&NiagaraSpriteRendererProperties]}
  {emitterName=FN_Contrails_SoupBubbles001, rendererClasses [&NiagaraSpriteRendererProperties]}
 ]
 userVariables [
  User.bIsFrontEnd {type=NiagaraBool}
  User.bIsFrontEndPreview {type=NiagaraBool}
  User.PlayerSpeed {type=NiagaraFloat, defaultValue=0, description=""}
  User.RibbonIDOffset {type=NiagaraInt32}
  User.SkeletalMesh {
   type=NiagaraDataInterfaceSkeletalMesh
   defaultValue {
    struct=NiagaraExt_VariableValue_DataInterface
    value {dataInterface=&self:NiagaraDataInterfaceSkeletalMesh_1}
   }
  }
 ]
}
